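# Modal widget in the Athena replay browser for viewing a streaming replay by entering a game session ID.
# Text changes and activation re-run IsValidGameSessionId to enable or disable the Accept button; activation also plays a fade and scale intro.
# Accept, or Enter in the field, passes the ID to the native base's PlayStreamingDSSReplay, with no Blueprint validity check on the Enter path.
# Close calls PopContentWidget.
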WidgetBlueprint /FortniteCoreUI/Athena/UI/ReplayBrowser/AthenaViewStreamingReplayModal
animations [Intro {
  binding_count=1
  length_seconds=0.25
  object=&self:Intro
  sequence {
   bindings [CommonBorder_1 {
     id=#1
     possessed_class=CommonBorder
     tracks [
      {
       type=MovieSceneFloatTrack
       display_name=Render Opacity
       sections [{type=MovieSceneFloatSection, end_frame=30, start_frame=0, channels [None {type=MovieSceneScriptingFloatChannel, keys [[0,0],[12,0],[30,1]]}]}]
      }
      {
       type=MovieScene2DTransformTrack
       sections [{
         type=MovieScene2DTransformSection
         end_frame=30
         start_frame=0
         channels [
          Scale.X {type=MovieSceneScriptingFloatChannel, keys [[12,1.15],[24,1]]}
          Scale.Y {type=MovieSceneScriptingFloatChannel, keys [[12,1.15],[24,1]]}
         ]
        }]
      }
     ]
    }]
   display_rate {numerator=120}
  }
 }]
graphs [
 EventGraph {nodes [
   "On Clicked (JoinButton)"!() -> (OutputDelegate, then, Button:Common Button Base Object Reference) {is=title K2Node_ComponentBoundEvent}
   n1 = Get EditableText_SessionSearch() -> (EditableText_SessionSearch:Editable Text Object Reference)
   > PlayStreamingDSSReplay!(
    execute
    self:Athena View Streaming Replay Modal Base Object Reference
    GameSessionId:String="To String (Text)"(InText="GetText (Editable Text)"!(self:Editable Text Object Reference=@n1) -> (ReturnValue:Text))
   ) -> (then)
   "On Clicked (CloseButton)"!() -> (OutputDelegate, then, Button:Common Button Base Object Reference) {is=title K2Node_ComponentBoundEvent}
   > PopContentWidget!(execute) -> (then)
   n2 = "OnTextChanged (Editable Text) (EditableText_SessionSearch)"!() -> (OutputDelegate, then, Text:"Text (by ref)") {is=title K2Node_ComponentBoundEvent}
   > UpdateJoinButtonEnabled!(execute, self, InText:Text=@n2.Text) -> (then)
   n3 = "OnTextCommitted (Editable Text) (EditableText_SessionSearch)"!() -> (OutputDelegate, then, Text:"Text (by ref)", CommitMethod:ETextCommit Enum=Default) {is=title K2Node_ComponentBoundEvent}
   > Branch(Condition="Equal (Enum)"(A:ETextCommit Enum=@n3.CommitMethod, B:ETextCommit Enum=OnEnter))
   > PlayStreamingDSSReplay(GameSessionId="To String (Text)"(InText=@n3.Text))
   Event On Activated()
   > PlayAnimationForward(InAnimation=Get Intro())
   > UpdateJoinButtonEnabled(InText="GetText (Editable Text)"(self=Get EditableText_SessionSearch()))
  ]}
 UpdateJoinButtonEnabled {nodes [
   n1 = UpdateJoinButtonEnabled!() -> (then, InText:Text) {is=title K2Node_FunctionEntry}
   > SetIsEnabled(
    self=Get JoinButton() -> (JoinButton:WBP UIKit Button Loud Object Reference)
    bInIsEnabled=IsValidGameSessionId!(self:Athena View Streaming Replay Modal Base Object Reference, GameSessionId:String="To String (Text)"(InText=@n1)) -> (ReturnValue:bool)
   )
  ]}
]
variables []
widgets {
 info {
  inheritedWidgetCount=1
  namedSlotCount=1
  parentClass=/Script/FortniteUI.AthenaViewStreamingReplayModalBase
  rootWidgetClass=&BackgroundBlur
  widgetCount=22
 }
 widgets [
  BackgroundBlur_0 {widgetClassPath=&BackgroundBlur, widget=BackgroundBlur_0, properties {blurStrength=15}}
  Lightbox {
   widgetClassPath=/Game/UI/Foundation/Lightbox/Lightbox
   parent=BackgroundBlur_0
   widget=Lightbox
   properties {dimAmount=0.75, backgroundColor=<rgba 0 0.0006 0.0075 1>}
   slot {class=BackgroundBlurSlot, properties {padding=<ltrb 0 0 0 0>}}
  }
  TapToCloseZone {
   widgetClassPath=&Border
   namedSlotHost=Lightbox
   widget=TapToCloseZone
   properties {
    horizontalAlignment=HAlign_Center
    verticalAlignment=VAlign_Center
    background {
     bIsDynamicallyLoaded=false
     drawAs=NoDrawType
     imageType=NoImage
     mirroring=NoMirror
     resourceName=None
     resourceObject=null
     tiling=NoTile
     imageSize=<32 32>
     margin=<ltrb 0 0 0 0>
     outlineSettings {
      bUseBrushTransparency=false
      roundingType=HalfHeightRadius
      width=0
      color {colorUseRule=UseColor_Specified, specifiedColor=<rgba 0 0 0 0>}
      cornerRadii=<0 0 0 0>
     }
     tintColor {colorUseRule=UseColor_Specified, specifiedColor=<rgba 1 1 1 1>}
     uVRegion {bIsValid=false, max=<0 0>, min=<0 0>}
    }
    padding=<ltrb 0 0 0 0>
   }
  }
  SizeBox_0 {
   widgetClassPath=&SizeBox
   parent=TapToCloseZone
   widget=SizeBox_0
   properties {
    bOverride_MinDesiredHeight=true
    bOverride_WidthOverride=true
    heightOverride=800
    minDesiredHeight=360
    minDesiredWidth=1000
    widthOverride=1000
   }
   slot {class=BorderSlot, properties {horizontalAlignment=HAlign_Center, verticalAlignment=VAlign_Center, padding=<ltrb 0 0 0 0>}}
  }
  CommonBorder_1 {
   widgetClassPath=&CommonBorder
   parent=SizeBox_0
   widget=CommonBorder_1
   properties {style=/FortUILibrary/Styles/Backgrounds/_BorderStyle_Modal._BorderStyle_Modal_C, padding=<ltrb 0 0 0 0>}
   slot {class=SizeBoxSlot}
  }
  Overlay_0 {widgetClassPath=&Overlay, parent=CommonBorder_1, widget=Overlay_0, slot {class=BorderSlot, properties {padding=<ltrb 0 0 0 0>}}}
  WBP_CaptureForPostBufferUpdate {
   widgetClassPath=&WBP_CaptureForPostBufferUpdate
   bIsVariable=true
   parent=Overlay_0
   slot {class=OverlaySlot, properties {horizontalAlignment=HAlign_Fill, verticalAlignment=VAlign_Fill}}
  }
  VerticalBox_1 {
   widgetClassPath=&VerticalBox
   parent=Overlay_0
   widget=VerticalBox_1
   slot {class=OverlaySlot, properties {horizontalAlignment=HAlign_Fill, verticalAlignment=VAlign_Fill}}
  }
  CommonBorder_Modal {
   widgetClassPath=&CommonBorder
   parent=VerticalBox_1
   widget=CommonBorder_Modal
   properties {style=/FortUILibrary/Styles/Backgrounds/_BorderStyle_ModalHeader._BorderStyle_ModalHeader_C, padding=<ltrb 40 0 40 0>}
   slot {class=VerticalBoxSlot}
  }
  HorizontalBox_1 {
   widgetClassPath=&HorizontalBox
   parent=CommonBorder_Modal
   widget=HorizontalBox_1
   slot {class=BorderSlot, properties {padding=<ltrb 40 0 40 0>}}
  }
  Spacer_0 {widgetClassPath=&Spacer, parent=HorizontalBox_1, widget=Spacer_0, properties {size=<0 85>}, slot {class=HorizontalBoxSlot}}
  Title {
   widgetClassPath=&CommonTextBlock
   parent=HorizontalBox_1
   widget=Title
   properties {
    style=/FortUILibrary/Styles/Typography/_TextStyle_SubH2._TextStyle_SubH2_C
    text=VIEW STREAMING REPLAY
    font {
     bForceMonospaced=false
     bMaterialIsStencil=false
     fontMaterial=null
     fontObject=&Burbank
     letterSpacing=0
     monospacedWidth=1
     size=33
     skewAmount=0
     typefaceFontName=BurbankRegular-Black
     outlineSettings {
      bApplyOutlineToDropShadows=false
      bMiteredCorners=false
      bSeparateFillAlpha=false
      outlineMaterial=null
      outlineSize=0
      outlineColor=<rgba 0 0 0 1>
     }
    }
    shadowOffset=<0 0>
    strikeBrush {
     bIsDynamicallyLoaded=false
     drawAs=NoDrawType
     imageType=NoImage
     mirroring=NoMirror
     resourceName=None
     resourceObject=null
     tiling=NoTile
     imageSize=<32 32>
     margin=<ltrb 0 0 0 0>
     outlineSettings {
      bUseBrushTransparency=false
      roundingType=HalfHeightRadius
      width=0
      color {colorUseRule=UseColor_Specified, specifiedColor=<rgba 0 0 0 0>}
      cornerRadii=<0 0 0 0>
     }
     tintColor {colorUseRule=UseColor_Specified, specifiedColor=<rgba 1 1 1 1>}
     uVRegion {bIsValid=false, max=<0 0>, min=<0 0>}
    }
   }
   slot {class=HorizontalBoxSlot, properties {verticalAlignment=VAlign_Center, padding=<ltrb 8 0 20 0>}}
  }
  Grid_Content {
   widgetClassPath=&GridPanel
   parent=VerticalBox_1
   widget=Grid_Content
   properties {columnFill [1], rowFill [1]}
   slot {class=VerticalBoxSlot, properties {verticalAlignment=VAlign_Top, padding=<ltrb 48 32 48 32>, size {sizeRule=Fill, value=1}}}
  }
  Spacer_Text {
   widgetClassPath=&Spacer
   parent=Grid_Content
   widget=Spacer_Text
   properties {size=<214 60>}
   slot {class=GridSlot, properties {column=1, horizontalAlignment=HAlign_Right, verticalAlignment=VAlign_Top}}
  }
  CommonTextBlock_3 {
   widgetClassPath=&CommonTextBlock
   parent=Grid_Content
   widget=CommonTextBlock_3
   properties {
    autoWrapText=true
    lineHeightPercentage=1.29
    style=/FortUILibrary/Styles/Typography/_TextStyle_Body1._TextStyle_Body1_C
    text=Enter a game session ID:
    font {
     bForceMonospaced=false
     bMaterialIsStencil=false
     fontMaterial=null
     fontObject=&BurbankSmall
     letterSpacing=0
     monospacedWidth=1
     size=18
     skewAmount=0
     typefaceFontName=Bold
     outlineSettings {
      bApplyOutlineToDropShadows=false
      bMiteredCorners=false
      bSeparateFillAlpha=false
      outlineMaterial=null
      outlineSize=0
      outlineColor=<rgba 0 0 0 1>
     }
    }
    shadowOffset=<0 0>
    strikeBrush {
     bIsDynamicallyLoaded=false
     drawAs=NoDrawType
     imageType=NoImage
     mirroring=NoMirror
     resourceName=None
     resourceObject=null
     tiling=NoTile
     imageSize=<32 32>
     margin=<ltrb 0 0 0 0>
     outlineSettings {
      bUseBrushTransparency=false
      roundingType=HalfHeightRadius
      width=0
      color {colorUseRule=UseColor_Specified, specifiedColor=<rgba 0 0 0 0>}
      cornerRadii=<0 0 0 0>
     }
     tintColor {colorUseRule=UseColor_Specified, specifiedColor=<rgba 1 1 1 1>}
     uVRegion {bIsValid=false, max=<0 0>, min=<0 0>}
    }
   }
   slot {class=GridSlot, properties {verticalAlignment=VAlign_Top}}
  }
  EditableText {
   widgetClassPath=&CommonBorder
   parent=Grid_Content
   widget=EditableText
   properties {
    verticalAlignment=VAlign_Center
    background {
     bIsDynamicallyLoaded=false
     drawAs=Image
     imageType=NoImage
     mirroring=NoMirror
     resourceName=None
     resourceObject=null
     tiling=NoTile
     imageSize=<32 32>
     margin=<ltrb 0 0 0 0>
     outlineSettings {
      bUseBrushTransparency=false
      roundingType=HalfHeightRadius
      width=0
      color {colorUseRule=UseColor_Specified, specifiedColor=<rgba 0 0 0 0>}
      cornerRadii=<0 0 0 0>
     }
     tintColor {colorUseRule=UseColor_Specified, specifiedColor=<rgba 1 1 1 1>}
     uVRegion {bIsValid=false, max=<0 0>, min=<0 0>}
    }
    padding=<ltrb 16 0 16 0>
   }
   slot {class=GridSlot, properties {layer=20, row=1, verticalAlignment=VAlign_Top, padding=<ltrb 0 0 40 0>}}
  }
  HorizontalBox_0 {
   widgetClassPath=&HorizontalBox
   parent=EditableText
   widget=HorizontalBox_0
   slot {class=BorderSlot, properties {verticalAlignment=VAlign_Center, padding=<ltrb 16 0 16 0>}}
  }
  EditableText_SessionSearch {
   widgetClassPath=/Script/UMG.EditableText
   bInherited=true
   bIsVariable=true
   parent=HorizontalBox_0
   widget=EditableText_SessionSearch
   properties {
    hintText=aabbccddeeff11223344556677889900
    navigation=&self:WidgetTree.EditableText_SessionSearch.WidgetNavigation_0
    widgetStyle {
     backgroundImageComposing {
      bIsDynamicallyLoaded=false
      drawAs=NoDrawType
      imageType=FullColor
      mirroring=NoMirror
      resourceName=../../../Engine/Content/Slate/Old/HyperlinkDotted.png
      resourceObject=null
      tiling=Both
      imageSize=<0 0>
      margin=<ltrb 0 0 0 0.1875>
      outlineSettings {
       bUseBrushTransparency=false
       roundingType=HalfHeightRadius
       width=0
       color {colorUseRule=UseColor_Specified, specifiedColor=<rgba 0 0 0 0>}
       cornerRadii=<0 0 0 0>
      }
      tintColor {colorUseRule=UseColor_Specified, specifiedColor=<rgba 1 1 1 1>}
      uVRegion {bIsValid=false, max=<0 0>, min=<0 0>}
     }
     backgroundImageSelected {
      bIsDynamicallyLoaded=false
      drawAs=NoDrawType
      imageType=FullColor
      mirroring=NoMirror
      resourceName=../../../Engine/Content/Slate/Common/EditableTextSelectionBackground.png
      resourceObject=null
      tiling=NoTile
      imageSize=<0 0>
      margin=<ltrb 0.25 0.25 0.25 0.25>
      outlineSettings {
       bUseBrushTransparency=false
       roundingType=HalfHeightRadius
       width=0
       color {colorUseRule=UseColor_Specified, specifiedColor=<rgba 0 0 0 0>}
       cornerRadii=<0 0 0 0>
      }
      tintColor {colorUseRule=UseColor_Specified, specifiedColor=<rgba 1 1 1 1>}
      uVRegion {bIsValid=false, max=<0 0>, min=<0 0>}
     }
     caretImage {
      bIsDynamicallyLoaded=false
      drawAs=Image
      imageType=FullColor
      mirroring=NoMirror
      resourceName=../../../Engine/Content/Slate/Old/White.png
      resourceObject=null
      tiling=NoTile
      imageSize=<16 16>
      margin=<ltrb 0 0 0 0>
      outlineSettings {
       bUseBrushTransparency=false
       roundingType=HalfHeightRadius
       width=0
       color {colorUseRule=UseColor_Specified, specifiedColor=<rgba 0 0 0 0>}
       cornerRadii=<0 0 0 0>
      }
      tintColor {colorUseRule=UseColor_Specified, specifiedColor=<rgba 1 1 1 1>}
      uVRegion {bIsValid=false, max=<0 0>, min=<0 0>}
     }
     colorAndOpacity {colorUseRule=UseColor_Specified, specifiedColor=<rgba 0 0 0 1>}
     font {
      bForceMonospaced=false
      bMaterialIsStencil=false
      fontMaterial=null
      fontObject=&BurbankSmall
      letterSpacing=0
      monospacedWidth=1
      size=16
      skewAmount=0
      typefaceFontName=Bold
      outlineSettings {
       bApplyOutlineToDropShadows=false
       bMiteredCorners=false
       bSeparateFillAlpha=false
       outlineMaterial=null
       outlineSize=0
       outlineColor=<rgba 0 0 0 1>
      }
     }
     hintColorAndOpacity {colorUseRule=UseColor_Foreground, specifiedColor=<rgba 1 0 1 1>}
     selectedBackgroundColor {colorUseRule=UseColor_Foreground, specifiedColor=<rgba 1 0 1 1>}
    }
   }
   slot {class=HorizontalBoxSlot, properties {verticalAlignment=VAlign_Center, size {sizeRule=Fill, value=1}}}
  }
  Spacer_EditText {
   widgetClassPath=&Spacer
   parent=HorizontalBox_0
   widget=Spacer_EditText
   properties {size=<0 56>}
   slot {class=HorizontalBoxSlot, properties {horizontalAlignment=HAlign_Left, verticalAlignment=VAlign_Top}}
  }
  VerticalBox_0 {
   widgetClassPath=&VerticalBox
   parent=Grid_Content
   widget=VerticalBox_0
   slot {class=GridSlot, properties {column=1, layer=30, row=1, rowSpan=2}}
  }
  JoinButton {
   widgetClassPath=/FortUILibrary/UIKit/Buttons/WBP_UIKit_Button_Loud
   bIsVariable=true
   parent=VerticalBox_0
   widget=JoinButton
   properties {text=ACCEPT, triggeringInputAction {dataTable=&InputActionDataTable, rowName=Accept}}
   slot {class=VerticalBoxSlot, properties {padding=<ltrb 0 0 0 24>}}
  }
  CloseButton {
   widgetClassPath=&WBP_UIKit_Button_Regular
   bIsVariable=true
   parent=VerticalBox_0
   widget=CloseButton
   properties {bHideInputActionWithKeyboard=true, text=CLOSE, triggeringInputAction {dataTable=&InputActionDataTable, rowName=Close}}
   slot {class=VerticalBoxSlot}
  }
 ]
}
